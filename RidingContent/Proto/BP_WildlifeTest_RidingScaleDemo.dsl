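; Debug demo actor for previewing riding at different creature scales, swapping one wildlife skeletal mesh between a wolf and a boar.
; Debug events adjust or reset scale and toggle creature or rider visibility, with RepNotify functions refreshing mesh, scale and rider visibility.
; SetPlayerRefPosition places the selected rider reference actor using a per-creature offset plus a calibration vector multiplied by scale minus one.
; BeginPlay fires both toggles, then positions the rider.
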(event EventBeginPlay
  (CallFunction|DebugWildlifeTestToggleCreature)
  (CallFunction|DebugWildlifeTestToggleHidePlayerRef)
  (CallFunction|SetPlayerRefPosition))

(event Collision|EventActorBeginOverlap
  (params (OtherActor :type Actor%20Object%20Reference)))

(event EventTick
  (params (DeltaSeconds :type Float%20%28single-precision%29)))

(event Custom|Debug_WildlifeTest_IncreaseScale
  (Variables|Default|SetcurrentScale (+ (Variables|Default|GetcurrentScale) (Variables|Default|GetincrementalDelta))))

(event Custom|Debug_WildlifeTest_DecreaseScale
  (bind _incrementaldelta (Variables|Default|GetincrementalDelta))
  (bind _currentscale (Variables|Default|GetcurrentScale))
  (bind _returnvalue (- _currentscale _incrementaldelta))
  (if (<= _returnvalue 0)
    (Variables|Default|SetcurrentScale 0.1)
    (else
      (Variables|Default|SetcurrentScale _returnvalue))))

(event Custom|Debug_WildlifeTest_ResetScale
  (Variables|Default|SetcurrentScale 1.0))

(event Custom|Debug_WildlifeTest_ToggleCreature
  (Variables|Default|SetcurrentSelectionWolf (not (Variables|Default|GetcurrentSelectionWolf))))

(event Custom|Debug_WildlifeTest_ToggleHidePlayerRef
  (Variables|Default|SetisPlayerVisible (not (Variables|Default|GetisPlayerVisible))))

(event Custom|UpdateWildlifeVis
  (if (Variables|Default|GetcurrentSelectionWolf)
    (Components|SkinnedMesh|SetSkinnedAssetandUpdate (Variables|Default|GetSkeletalMesh_Wildlife) "/Irwin/AI/Predators/Grandma/Meshes/Grandma_Mammal.Grandma_Mammal" true)
    (if (Variables|Default|GetisPlayerVisible)
      (Rendering|SetActorHiddenInGame (Variables|Default|GetPlayerRiderRef_Wolf) false)
      (Rendering|SetActorHiddenInGame (Variables|Default|GetPlayerRiderRef_Boar) true))
    (else
      (Components|SkinnedMesh|SetSkinnedAssetandUpdate (Variables|Default|GetSkeletalMesh_Wildlife) "/Irwin/AI/Prey/Burt/Meshes/Burt_Mammal.Burt_Mammal" true)
      (if (Variables|Default|GetisPlayerVisible)
        (Rendering|SetActorHiddenInGame (Variables|Default|GetPlayerRiderRef_Wolf) true)
        (Rendering|SetActorHiddenInGame (Variables|Default|GetPlayerRiderRef_Boar) false)))))

(fn OnRep_currentScale
  (bind _currentscale (Variables|Default|GetcurrentScale))
  (Transformation|SetWorldScale3D (Variables|Default|GetSkeletalMesh_Wildlife) (Math|Vector|MakeVector _currentscale _currentscale _currentscale))
  (CallFunction|SetPlayerRefPosition)
  (Development|PrintString (Utilities|String|Append "Current Wildlife Scale Is:" (Utilities|String|ToString(Float) _currentscale))))

(fn OnRep_currentSelectionWolf
  (CallFunction|UpdateWildlifeVis)
  (CallFunction|SetPlayerRefPosition))

(fn OnRep_isPlayerVisible
  (if (not (Variables|Default|GetisPlayerVisible))
    (Rendering|SetActorHiddenInGame (Variables|Default|GetPlayerRiderRef_Wolf) true)
    (Rendering|SetActorHiddenInGame (Variables|Default|GetPlayerRiderRef_Boar) true))
  (CallFunction|UpdateWildlifeVis)
  (CallFunction|SetPlayerRefPosition))

(fn SetPlayerRefPosition
  (bind _currentselectionwolf (Variables|Default|GetcurrentSelectionWolf))
  (bind _playerriderref_wolf (Variables|Default|GetPlayerRiderRef_Wolf))
  (bind _playerriderref_boar (Variables|Default|GetPlayerRiderRef_Boar))
  (bind _returnvalue (select _currentselectionwolf _playerriderref_wolf _playerriderref_boar))
  (Transformation|SetActorLocation _returnvalue (Math|Transform|TransformLocation (Transformation|GetActorTransform self) (+ (Math|Vector|SelectVector (Variables|Default|GetWolfAttach_Offset) (Variables|Default|GetBoarAttach_Offset) _currentselectionwolf) (* (Math|Vector|SelectVector (Variables|Default|GetWolfAttach_OffsetCalibration) (Variables|Default|GetBoarAttach_OffsetCalibration) _currentselectionwolf) (- (Variables|Default|GetcurrentScale) 1.0)))) false true)
  (Transformation|SetActorRotation _returnvalue (Transformation|GetActorRotation self) false))

(fn ConstructionScript)
